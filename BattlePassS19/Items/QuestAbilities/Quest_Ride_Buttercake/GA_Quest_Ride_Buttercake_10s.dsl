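; Gameplay ability for the Ride Buttercake quest that credits the player after a set duration, 10 seconds going by the asset name.
; Activation from an event starts a looping timer that adds its interval to an elapsed time counter, with no riding check in this graph.
; Reaching the required time sends a tagged custom stat event to the owning player's quest manager and ends the ability.
; Ending clears the timer and resets the counter.
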(event Ability|EventActivateAbilityFromEvent
  (params (EventData :type Gameplay%20Event%20Data%20Structure%20%28by%20ref%29))
  (|CallParentFunction EventData)
  ;[1 1 1] Start Checking Vehicle If Vehicle is in Air
  (bind _returnvalue (Utilities|Time|SetTimerbyEvent (delegate Custom|Check%20Time) (Variables|Default|GetExecutionTimeInterval) true))
  (Variables|Default|SetTimer%20Handle _returnvalue))

(event Custom|Check%20Time
  ;[1 1 1] Add time every interval
  (bind _output_get (Variables|Default|SetElapsedTime (+ (Variables|Default|GetExecutionTimeInterval) (Variables|Default|GetElapsedTime))))
  (if (>= _output_get (Variables|Default|GetRequiredTime))
    (Quest|SendComplexCustomStatEvent (Variables|GAVars|Data|GetOwningPlayerQuestManager) 0 (Variables|Default|GetT_Quest) "(GameplayTags=)" 1)
    (Ability|EndAbility)))

(event Ability|EventOnEndAbility
  (params (bWasCancelled :type Boolean))
  ;[1 1 1] On Ability End, reset timer and clear handle
  (Utilities|Time|ClearandInvalidateTimerbyHandle (Variables|Default|GetTimer%20Handle))
  (Variables|Default|SetElapsedTime 0.0)
  (Ability|EndAbility))
